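# Sound cue for the thrust start sound of the Galileo ZeppelinFemale glider cosmetic.
# A Random node picks between two ThrustStart wave variations with equal weight.
# Playback runs slightly lowered in pitch and volume, using the glider thrust attenuation, a thrust concurrency group, a glider sound class and the cosmetic glider loop audio properties sheet.
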SoundCue /BRCosmetics/Sounds/Gliders/Galileo_ZeppelinFemale/Galileo_ZeppelinFemale_Thrust_Cue
root=Random_0
nodes [
 {id=Random_0, children [WavePlayer_0,WavePlayer_2], properties {weights [1,1]}}
 {id=WavePlayer_0, properties {soundWaveAssetPtr=&/Galileo_ZeppelinFemale_ThrustStart_01}}
 {id=WavePlayer_2, properties {soundWaveAssetPtr=&/Galileo_ZeppelinFemale_ThrustStart_02}}
]
properties {
 attenuationSettings=&Fort_Player_Glider_Thrust_Athena_Att
 audioPropertiesSheet=&APS_Cosmetic_Glider_Loop
 maxDistance=12500
 pitchMultiplier=0.9
 soundClassObject=&Fort_Gliders_Land
 volumeMultiplier=0.6
 concurrencySet [&Athena_Glider_Thrust_Conc]
}
